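; Interactable Watering Hole instrument that shows a "Play" prompt and toggles hidden music.
; Interaction flips the HiddenMusicActivated flag, gated by a CanThePlayerInteract variable.
; The flag's OnRep passes the state to the music controller's EnableHiddenMusic only on remote, non-authority instances.
; BeginPlay flushes net dormancy, and the interaction icon sits slightly above the actor.
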(event Interaction|EventBlueprintOnInteract
  (params (InteractingPawn :type Fort%20Pawn%20Object%20Reference) (InteractionBeingAttempted :type EInteractionBeingAttempted%20Enum :default "FirstInteraction"))
  (Variables|Default|SetHiddenMusicActivated (not (Variables|Default|GetHiddenMusicActivated))))

(event EventBeginPlay
  (Networking|FlushNetDormancy))

(fn BlueprintCanInteract
  (params (InteractingPawn :type Fort%20Pawn%20Object%20Reference) (InteractionBeingAttempted :type EInteractionBeingAttempted%20Enum :default "FirstInteraction") (InteractionType :type TInteractionType%20Enum :default "IT_NoInteraction"))
  (return (Variables|Default|GetCanThePlayerInteract)))

(fn BlueprintGetInteractionString
  (params (InteractingPawn :type Fort%20Pawn%20Object%20Reference) (InteractionBeingAttempted :type EInteractionBeingAttempted%20Enum :default "FirstInteraction"))
  (return "NSLOCTEXT(\"[AE523A5B4B9135723412B2BA51227EEB]\", \"BCD2B2024A08C5D1F616D3A569780A9E\", \"Play\")"))

(fn GetIconPlacement
  (params (SelfActor :type Actor%20Object%20Reference) (ViewingActor :type Actor%20Object%20Reference))
  (return (+ (Transformation|GetActorLocation SelfActor) "0, 0,50.000000") "0, 0, 0"))

(fn OnRep_HiddenMusicActivated
  (Networking|SwitchHasAuthority
    (:Remote
      (Class|BPWateringHoleAmbTrigger|EnableHiddenMusic (Variables|Default|GetMusicController) (Variables|Default|GetHiddenMusicActivated)))))

(fn ConstructionScript)
